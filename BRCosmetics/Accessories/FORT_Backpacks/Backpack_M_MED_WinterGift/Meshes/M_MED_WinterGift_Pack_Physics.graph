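# Physics asset for the WinterGift backpack cosmetic, defining secondary motion for its wrapping paper, bow and ribbon.
# Capsule bodies on dyn_ bones hang from base_ anchors through swing and twist limits: paper flaps per bottom, middle and top tier, plus a ribbon branching into left and right chains.
# Each bow loop is constrained to both dyn_bow and base_bow, and base_mid_paper carries an extra large, angled capsule.
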PhysicsAsset /BRCosmetics/Accessories/FORT_Backpacks/Backpack_M_MED_WinterGift/Meshes/M_MED_WinterGift_Pack_Physics
bodies {
 base_bot_paper [{shapeType=Capsule, length=1, radius=1, shapeName=base_bot_paper_capsule, rotation=<rot 90 0 0>}]
 base_bow [{shapeType=Capsule, length=1, radius=1, shapeName=base_bow_capsule, rotation=<rot 90 0 0>}]
 base_mid_paper [
  {shapeType=Capsule, length=1, radius=1, shapeName=base_mid_paper_capsule, rotation=<rot 90 0 0>}
  {
   shapeType=Capsule
   length=25
   radius=4
   shapeName=base_mid_paper_capsule
   center=<-2.6821 -3.4759 -2.7263>
   rotation=<rot 85.7492 -124.8974 -179.63>
  }
 ]
 base_ribbon [{shapeType=Capsule, length=1, radius=1, shapeName=base_top_paper_capsule, rotation=<rot 90 0 0>}]
 base_top_paper [{shapeType=Capsule, length=1, radius=1, shapeName=base_top_paper_capsule, rotation=<rot 90 0 0>}]
 dyn_bot_paper_01 [{shapeType=Capsule, length=2, radius=1, shapeName=dyn_bot_paper_01_capsule, center=<1 0 0>, rotation=<rot 90 0 0>}]
 dyn_bot_paper_02a [{shapeType=Capsule, length=2, radius=1, shapeName=dyn_bot_paper_02a_capsule, center=<1 0 0>, rotation=<rot 90 0 0>}]
 dyn_bot_paper_02b [{shapeType=Capsule, length=2, radius=1, shapeName=dyn_bot_paper_02b_capsule, center=<1 0 0>, rotation=<rot 90 0 0>}]
 dyn_bot_paper_03 [{shapeType=Capsule, length=2, radius=1, shapeName=dyn_bot_paper_03_capsule, center=<1 0 0>, rotation=<rot 90 0 0>}]
 dyn_bow [{shapeType=Capsule, length=3, radius=2, shapeName=dyn_bow_capsule, center=<1 0 0>, rotation=<rot 90 0 0>}]
 dyn_bow_l [
  {shapeType=Capsule, length=3, radius=2, shapeName=dyn_bow_l_capsule, center=<2 0 0>, rotation=<rot 90 0 0>}
  {shapeType=Capsule, length=5, radius=2, shapeName=dyn_bow_l_capsule, center=<4.292 0 0>}
 ]
 dyn_bow_r [{shapeType=Capsule, length=3, radius=2, shapeName=dyn_bow_r_capsule, center=<1.5 0 0>, rotation=<rot 90 0 0>}]
 dyn_mid_paper_01 [{shapeType=Capsule, length=2, radius=1, shapeName=dyn_mid_paper_01_capsule, center=<1 0 0>, rotation=<rot 90 0 0>}]
 dyn_mid_paper_02 [{shapeType=Capsule, length=2, radius=1, shapeName=dyn_mid_paper_02_capsule, center=<1 0 0>, rotation=<rot 90 0 0>}]
 dyn_mid_paper_03 [{shapeType=Capsule, length=2, radius=1, shapeName=dyn_mid_paper_03_capsule, center=<1 0 0>, rotation=<rot 90 0 0>}]
 dyn_mid_paper_04 [{shapeType=Capsule, length=2, radius=1, shapeName=dyn_mid_paper_04_capsule, center=<1 0 0>, rotation=<rot 90 0 0>}]
 dyn_ribbon [{shapeType=Capsule, length=2, radius=1.5, shapeName=dyn_ribbon_l_01_capsule, center=<1 0 0>, rotation=<rot 90 0 0>}]
 dyn_ribbon_l [{shapeType=Capsule, length=3, radius=1.5, shapeName=dyn_ribbon_l_01_capsule, center=<1.5 0 0>, rotation=<rot 90 0 0>}]
 dyn_ribbon_l_01 [{shapeType=Capsule, length=2.25, radius=1.5, shapeName=dyn_ribbon_l_01_capsule, center=<1 0 0>, rotation=<rot 90 0 0>}]
 dyn_ribbon_l_02 [{shapeType=Capsule, length=3, radius=1.75, shapeName=dyn_ribbon_l_02_capsule, center=<1.5 0 0>, rotation=<rot 90 0 0>}]
 dyn_ribbon_l_03 [{shapeType=Capsule, length=4, radius=1.75, shapeName=dyn_ribbon_l_03_capsule, center=<1.4994 -0.0846 0>, rotation=<rot 90 180 -177.5916>}]
 dyn_ribbon_r [{shapeType=Capsule, length=4.5, radius=1.5, shapeName=dyn_ribbon_r_01_capsule, center=<1.7797 0 0>, rotation=<rot 90 0 0>}]
 dyn_ribbon_r_01 [{shapeType=Capsule, length=2, radius=1.5, shapeName=dyn_ribbon_r_01_capsule, center=<1 0 0>, rotation=<rot 90 0 0>}]
 dyn_ribbon_r_02 [{shapeType=Capsule, length=2, radius=1.75, shapeName=dyn_ribbon_r_02_capsule, center=<1 0 0>, rotation=<rot 90 0 0>}]
 dyn_ribbon_r_03 [{shapeType=Capsule, length=2.5, radius=1.75, shapeName=dyn_ribbon_r_03_capsule, center=<1.25 0 0>, rotation=<rot 90 0 0>}]
 dyn_ribbon_r_04 [{shapeType=Capsule, length=4.25, radius=2, shapeName=dyn_ribbon_r_04_capsule, center=<2.4289 0.3351 0>, rotation=<rot 90 -176.4237 175.0958>}]
 dyn_top_paper_01 [{shapeType=Capsule, length=2, radius=1, shapeName=dyn_top_paper_01_capsule, center=<1 0 0>, rotation=<rot 90 0 0>}]
 dyn_top_paper_02 [{shapeType=Capsule, length=2, radius=1, shapeName=dyn_top_paper_02_capsule, center=<1 0 0>, rotation=<rot 90 0 0>}]
 dyn_top_paper_03 [{shapeType=Capsule, length=2, radius=1, shapeName=dyn_top_paper_03_capsule, center=<1 0 0>, rotation=<rot 90 0 0>}]
 dyn_top_paper_04 [{shapeType=Capsule, length=2, radius=1, shapeName=dyn_top_paper_04_capsule, center=<1 0 0>, rotation=<rot 90 0 0>}]
}
constraints [
 {bone1Name=dyn_bot_paper_01, bone2Name=base_bot_paper, swing1LimitDegrees=15, swing2LimitDegrees=25, twistLimitDegrees=5}
 {bone1Name=dyn_bot_paper_03, bone2Name=base_bot_paper, swing1LimitDegrees=15, swing2LimitDegrees=25, twistLimitDegrees=5}
 {bone1Name=dyn_bot_paper_02a, bone2Name=base_bot_paper, swing1LimitDegrees=15, swing2LimitDegrees=15, twistLimitDegrees=5}
 {bone1Name=dyn_bot_paper_02b, bone2Name=dyn_bot_paper_02a, swing1LimitDegrees=15, swing2LimitDegrees=10, twistLimitDegrees=5}
 {bone1Name=dyn_bow, bone2Name=base_bow, swing1LimitDegrees=1, swing2LimitDegrees=1, twistLimitDegrees=10}
 {bone1Name=dyn_bow_l, bone2Name=dyn_bow, swing1LimitDegrees=26, swing2LimitDegrees=40, twistLimitDegrees=15}
 {bone1Name=dyn_bow_r, bone2Name=dyn_bow, swing1LimitDegrees=26, swing2LimitDegrees=40, twistLimitDegrees=15}
 {bone1Name=dyn_mid_paper_01, bone2Name=base_mid_paper, swing1LimitDegrees=30, swing2LimitDegrees=20, twistLimitDegrees=5}
 {bone1Name=dyn_mid_paper_02, bone2Name=base_mid_paper, swing1LimitDegrees=20, swing2LimitDegrees=25, twistLimitDegrees=5}
 {bone1Name=dyn_mid_paper_03, bone2Name=base_mid_paper, swing1LimitDegrees=20, swing2LimitDegrees=25, twistLimitDegrees=5}
 {bone1Name=dyn_mid_paper_04, bone2Name=base_mid_paper, swing1LimitDegrees=25, swing2LimitDegrees=20, twistLimitDegrees=5}
 {bone1Name=dyn_ribbon_l_02, bone2Name=dyn_ribbon_l_01, swing1LimitDegrees=45, swing2LimitDegrees=40, twistLimitDegrees=5}
 {bone1Name=dyn_ribbon_l_03, bone2Name=dyn_ribbon_l_02, swing1LimitDegrees=35, swing2LimitDegrees=30, twistLimitDegrees=10}
 {bone1Name=dyn_ribbon_r_02, bone2Name=dyn_ribbon_r_01, swing1LimitDegrees=30, swing2LimitDegrees=25, twistLimitDegrees=5}
 {bone1Name=dyn_ribbon_r_03, bone2Name=dyn_ribbon_r_02, swing1LimitDegrees=30, swing2LimitDegrees=35, twistLimitDegrees=5}
 {bone1Name=dyn_ribbon_r_04, bone2Name=dyn_ribbon_r_03, swing1LimitDegrees=25, swing2LimitDegrees=30, twistLimitDegrees=5}
 {bone1Name=dyn_top_paper_01, bone2Name=base_top_paper, swing1LimitDegrees=10, swing2LimitDegrees=25, twistLimitDegrees=5}
 {bone1Name=dyn_top_paper_02, bone2Name=base_top_paper, swing1LimitDegrees=10, swing2LimitDegrees=25, twistLimitDegrees=5}
 {bone1Name=dyn_top_paper_03, bone2Name=base_top_paper, swing1LimitDegrees=10, swing2LimitDegrees=25, twistLimitDegrees=5}
 {bone1Name=dyn_top_paper_04, bone2Name=base_top_paper, swing1LimitDegrees=10, swing2LimitDegrees=25, twistLimitDegrees=5}
 {bone1Name=dyn_ribbon_l_01, bone2Name=dyn_ribbon_l, swing1LimitDegrees=40, swing2LimitDegrees=35, twistLimitDegrees=5}
 {bone1Name=dyn_ribbon_r_01, bone2Name=dyn_ribbon_r, swing1LimitDegrees=25, swing2LimitDegrees=25, twistLimitDegrees=5}
 {bone1Name=dyn_ribbon, bone2Name=base_ribbon, swing1LimitDegrees=25, swing2LimitDegrees=20, twistLimitDegrees=5}
 {bone1Name=dyn_ribbon_l, bone2Name=dyn_ribbon, swing1LimitDegrees=20, swing2LimitDegrees=25, twistLimitDegrees=5}
 {bone1Name=dyn_ribbon_r, bone2Name=dyn_ribbon, swing1LimitDegrees=30, swing2LimitDegrees=20, twistLimitDegrees=5}
 {bone1Name=dyn_bow_l, bone2Name=base_bow, swing1LimitDegrees=26, swing2LimitDegrees=40, twistLimitDegrees=35}
 {bone1Name=dyn_bow_r, bone2Name=base_bow, swing1LimitDegrees=20, swing2LimitDegrees=60, twistLimitDegrees=25}
]
